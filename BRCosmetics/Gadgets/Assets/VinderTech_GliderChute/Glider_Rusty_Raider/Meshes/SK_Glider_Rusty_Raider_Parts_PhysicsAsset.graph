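# Physics asset for the parts skeletal mesh of the Rusty Raider glider cosmetic, giving it simplified collision and jointed bodies.
# Five box bodies cover the frame, left and right wings, parachute, and a tiny Disc_Top hub.
# Constraints form a tree from Disc_Top to frame and parachute, with both wings hanging off the frame, all sharing identical swing and twist limit angles.
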PhysicsAsset /BRCosmetics/Gadgets/Assets/VinderTech_GliderChute/Glider_Rusty_Raider/Meshes/SK_Glider_Rusty_Raider_Parts_PhysicsAsset
bodies {
 Disc_Top [{shapeType=Box, extentX=1.01, extentY=1.01, extentZ=1.01, rotation=<rot 0 -90 0>}]
 Frame [{shapeType=Box, extentX=103.652, extentY=31.0537, extentZ=149.8878, center=<0.0647 34.7741 4.3878>, rotation=<rot 0 90.0932 -89.9874>}]
 Wing_L [{shapeType=Box, extentX=20.2792, extentY=47.3617, extentZ=66.0564, center=<-13.8528 -5.6076 -20.7563>, rotation=<rot -5.201 -90 21.2056>}]
 Wing_R [{shapeType=Box, extentX=19.5029, extentY=44.547, extentZ=73.0606, center=<-13.4763 -5.2388 18.5302>, rotation=<rot 7.3523 -90 -27.6943>}]
 parachute [{shapeType=Box, extentX=100.1251, extentY=44.006, extentZ=135.4845, center=<-0.6917 10.1507 -1.3205>, rotation=<rot 0 0.9148 -86.1326>}]
}
constraints [
 {bone1Name=Frame, bone2Name=Disc_Top, swing1LimitDegrees=45, swing2LimitDegrees=45, twistLimitDegrees=45}
 {bone1Name=Wing_L, bone2Name=Frame, swing1LimitDegrees=45, swing2LimitDegrees=45, twistLimitDegrees=45}
 {bone1Name=Wing_R, bone2Name=Frame, swing1LimitDegrees=45, swing2LimitDegrees=45, twistLimitDegrees=45}
 {bone1Name=parachute, bone2Name=Disc_Top, swing1LimitDegrees=45, swing2LimitDegrees=45, twistLimitDegrees=45}
]
